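# One-shot stereo MetaSound source for the third-person launch pad burst sound, in the Fort_Gadgets sound class.
# OnPlay triggers a single Wave Player for the burst wave, which feeds both outputs and signals OnFinished when it ends.
# Character, movement and attenuation distance interface inputs are declared but not wired.
# Spatial falloff comes from a dedicated burst attenuation asset, and a concurrency override caps simultaneous instances.
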MetaSoundSource /Hermes/Audio/LaunchPad/MSS_LaunchPad_3P_Burst
inputs [
 Character.Affiliation {type=Enum:FortAudioAffiliation}
 Character.Health {type=Float}
 Character.UniqueID {type=Int32}
 Movement.Speed {type=Float}
 Movement.SpeedAzimuth {type=Float}
 Movement.SpeedElevation {type=Float}
 UE.Attenuation.Distance {type=Float}
 UE.Source.OnPlay {type=Trigger}
]
interfaces [Fort.Character:1.0, Fort.Movement:1.0, UE.Attenuation:1.0, UE.OutputFormat.Stereo:1.0, UE.Source.OneShot:1.0, UE.Source:1.1]
outputs [
 "UE.OutputFormat.Stereo.Audio:0" {type=Audio}
 "UE.OutputFormat.Stereo.Audio:1" {type=Audio}
 UE.Source.OneShot.OnFinished {type=Trigger}
]
pages [Default {
  input_defaults {
   Character.Affiliation=0
   Character.Health=0
   Character.UniqueID=0
   Movement.Speed=0
   Movement.SpeedAzimuth=0
   Movement.SpeedElevation=0
   UE.Attenuation.Distance=8000
   UE.Source.OnPlay=null
  }
  nodes [WavePlayer_0 = UE.Wave Player:Stereo(Play=^UE.Source.OnPlay) {also {defaults {Wave Asset=&/LaunchPad_3P_Burst}}}]
  links [
   {dst="@UE.Source.OneShot.OnFinished", src=WavePlayer_0.On Finished}
   {dst="@UE.OutputFormat.Stereo.Audio:0", src=WavePlayer_0.Out Left}
   {dst="@UE.OutputFormat.Stereo.Audio:1", src=WavePlayer_0.Out Right}
  ]
 }]
properties {
 assetImportData=&self:AssetImportData
 attenuationSettings=&/ATT_LaunchPad_Burst
 bOverrideConcurrency=true
 maxDistance=15000
 numChannels=2
 outputFormat=Stereo
 sampleRate=48000
 soundClassObject=&Fort_Gadgets
 volume=2
 concurrencyOverrides {maxCount=2}
}
